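# Niagara idle effect for the GainSneeze guitar cosmetic that sheds falling sand-like bits and dust sprites.
# BitsHigh and PixieDust derive spawn probability from owner speed, peaking at moderate motion and fading at rest or high speed.
# User.VariantID switches particle colors, while PixieDust also varies lifetime with speed, blends world/local space and kills particles by a speed threshold.
# Two fog emitters are present but disabled.
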NiagaraSystem /SparksCosmetics/Instrument/Guitar/GainSneeze/Effects/NS_Guitar_GainSneeze_Idle_Sand
dependencies {
 usedDataInterfaces [
  &NiagaraDataInterfaceCurve
  &NiagaraDataInterfaceArrayColor
  &NiagaraDataInterfaceSpriteRendererInfo
  &NiagaraDataInterfaceVector2DCurve
 ]
 usedDynamicInputs [&Vector2DFromCurve]
 usedModules [&SystemState, &BlendSpaceHelper]
}
emitters {
 BitsHigh {
  emitterUpdateScript [
   &EmitterState()
   &SpawnBurst_Instantaneous(Spawn Count=6, Use Spawn Probability=false, UseLoopCountLimit=false) {enabled=false}
   &SpawnRate(
    Spawn Probability=&FloatFromCurve(
     FloatCurve=DataInterface() {propertyValues {Curve {keys [
         {interpMode=RCIM_Cubic, leaveTangentWeight=0.1731, tangentMode=RCTM_User, time=0.15, value=0}
         {interpMode=RCIM_Cubic, tangentMode=RCTM_User, time=0.65, value=1}
         {arriveTangentWeight=0.6283, interpMode=RCIM_Cubic, tangentMode=RCTM_User, time=1, value=0}
        ]}}}
     CurveIndex=&RemapRange(
      Input Value To Remap=&VectorLength(Vector=$Engine.Owner.Velocity)
      Input Min=0
      Input Max=200
      Output Min=0
      Output Max=1
     )
    )
    SpawnRate=30
    Use Spawn Probability=true
   )
   &SpawnRate(SpawnRate=10, Use Spawn Probability=true) {enabled=false, module=SpawnRate001}
  ]
  particleSpawnScript [
   &InitializeParticle(
    Color=&MergeHelper_ColorSwitch(
     0=&SelectLinearColorFromArray(
      Array Sampling Mode=::NewEnumerator0
      Direct Array Index=?
      Color Selection Array=DataInterface() {propertyValues {ColorData [<rgba 1 0.647 0 1>, <rgba 0 0.474 0.3083 1>, <rgba 0.449 0 0 1>, <rgba 0.8878 1 0.7183 1>, <rgba 0.8878 1 0.7183 1>]}}
      Random Seed=0
     )
     1=<rgba 0 0.0573 0.0179 1>
     2=?
     3=?
     Index=&MakeIntFromFloat(Float=$User.VariantID)
     Num=::NewEnumerator0
    )
    Lifetime=&RandomRangeFloat(Minimum=1, Maximum=1.5, Evaluation Type=::NewEnumerator1)
    Sprite Size Mode=::NewEnumerator1
    Uniform Sprite Size Max=1.5
    Uniform Sprite Size Min=0.5
    UsePositionOffset=false
    Write Material Random=true
   )
   &ShapeLocation(MaskSpawnGroup=false, Non Uniform Scale=<1 0.1 1>, Override Seed=false, Sphere Radius=5, UseOwnerScale=true)
   &AddVelocity(Module.Use Falloff Distance=false, Override Seed=false, Velocity Mode=::NewEnumerator1, Velocity Speed=10)
   &CameraOffset(Camera Offset Amount=-10)
  ]
  particleUpdateScript [
   &ParticleState()
   &AccelerationForce(Acceleration=&RandomRangeVector(Minimum=<-15 -15 -45>, Maximum=<15 15 -20>, Evaluation Type=::NewEnumerator1))
   &SolveForcesAndVelocity(Clamp Velocity=false, Limit Acceleration=false)
   &ScaleColor(
    Scale Alpha=&FloatFromCurve(FloatCurve=DataInterface() {propertyValues {Curve {keys [<at 0 1 RCIM_Cubic>, {arriveTangent=-1, interpMode=RCIM_Cubic, leaveTangent=-1, time=0.505, value=0.8029}, <at 1 0>]}}})
    ScaleA=true
    ScaleRGB=false
   )
  ]
  rendererClasses [&NiagaraSpriteRendererProperties]
  renderers [{rendererClass=&NiagaraSpriteRendererProperties}]
 }
 FogCenterHigh {
  bEnabled=false
  emitterUpdateScript [
   &EmitterState()
   &SpawnRate(
    SpawnRate=&RandomRangeFloat(Minimum=4, Maximum=6, Evaluation Type=::NewEnumerator1)
    Use Spawn Probability=false
   )
  ]
  particleSpawnScript [&InitializeParticle(
    Color=<rgba 7.188 10 2.9391 0.03>
    Lifetime=&RandomRangeFloat(Minimum=0.75, Maximum=1.25, Evaluation Type=::NewEnumerator1)
    Sprite Rotation Mode=::NewEnumerator2
    Sprite Size=&Multiply_Vector2D_ByFloat(
     Float=&RandomRangeFloat(Minimum=50, Maximum=60)
     Vector2D=&RandomRangeVector2D(Minimum=<0.5 0.5>, Maximum=<1 1>)
    )
    Sprite Size Mode=::NewEnumerator4
    UsePositionOffset=false
    Write Material Random=true
   )]
  particleUpdateScript [
   &ParticleState()
   &SpriteRotationRate(Rotation Rate=&RandomRangeFloat(Minimum=-150, Maximum=0))
   &AccelerationForce(Acceleration=&RandomRangeVector(Minimum=<-15 -15 -45>, Maximum=<15 15 -30>, Evaluation Type=::NewEnumerator1))
   &SolveForcesAndVelocity(Clamp Velocity=false, Limit Acceleration=false)
   &ScaleColor(
    Scale Alpha=&FloatFromCurve(FloatCurve=DataInterface() {propertyValues {Curve {keys [<at -0.0003 0 RCIM_Cubic>,<at 0.1829 1 RCIM_Cubic>,<at 1 0>]}}})
    ScaleA=true
    ScaleRGB=false
   )
   &ScaleSpriteSize(Uniform Curve Sprite Scale=DataInterface() {propertyValues {Curve {keys [
        {arriveTangent=1.8174, interpMode=RCIM_Cubic, leaveTangent=1.8174, tangentMode=RCTM_User, time=0, value=0}
        {tangentMode=RCTM_User, time=1, value=1}
       ]}}})
   &CameraOffset(Camera Offset Amount=-5)
   &SubUVAnimation(
    Random Start Frame=true
    "Sprite Renderer "=DataInterface() {propertyValues {SpriteRenderer=&self:FogCenterHigh_0.NiagaraSpriteRendererProperties_0}}
    SubUV Lookup Index Scale:NiagaraFloat=&RandomRangeFloat(Minimum=0.2, Maximum=0.4)
    UseEndFrame=false
    UseStartFrame=false
   )
  ]
  rendererClasses [&NiagaraSpriteRendererProperties]
  renderers [{rendererClass=&NiagaraSpriteRendererProperties}]
 }
 FogHigh {
  bEnabled=false
  emitterUpdateScript [
   &EmitterState()
   &SpawnRate(
    SpawnRate=&RandomRangeFloat(Minimum=4, Maximum=6, Evaluation Type=::NewEnumerator1)
    Use Spawn Probability=false
   )
  ]
  particleSpawnScript [&InitializeParticle(
    Color=<rgba 3.4832 5 1.1913 1>
    Lifetime=&RandomRangeFloat(Minimum=0.75, Maximum=1.25, Evaluation Type=::NewEnumerator1)
    Sprite Rotation Mode=::NewEnumerator2
    Sprite Size Mode=::NewEnumerator1
    Uniform Sprite Size Max=80
    Uniform Sprite Size Min=60
    UsePositionOffset=false
    Write Material Random=true
   )]
  particleUpdateScript [
   &ParticleState()
   &SpriteRotationRate(Rotation Rate=&RandomRangeFloat(Minimum=-150, Maximum=0))
   &AccelerationForce(Acceleration=&RandomRangeVector(Minimum=<-15 -15 -45>, Maximum=<15 15 -20>, Evaluation Type=::NewEnumerator1))
   &SolveForcesAndVelocity(Clamp Velocity=false, Limit Acceleration=false)
   &ScaleColor(
    Scale Alpha=&FloatFromCurve(FloatCurve=DataInterface() {propertyValues {Curve {keys [<at -0.0003 0 RCIM_Cubic>,<at 0.1829 1 RCIM_Cubic>,<at 1 0>]}}}, Scale Curve=0.02)
    ScaleA=true
    ScaleRGB=false
   )
   &ScaleSpriteSize()
   &CameraOffset(Camera Offset Amount=-5)
   &SubUVAnimation(
    Random Start Frame=true
    "Sprite Renderer "=DataInterface() {propertyValues {SpriteRenderer=&self:FogHigh_0.NiagaraSpriteRendererProperties_0}}
    SubUV Lookup Index Scale:NiagaraFloat=&RandomRangeFloat(Minimum=0.1, Maximum=0.3)
    UseEndFrame=false
    UseStartFrame=false
   )
  ]
  rendererClasses [&NiagaraSpriteRendererProperties]
  renderers [{rendererClass=&NiagaraSpriteRendererProperties}]
 }
 PixieDust {
  emitterUpdateScript [
   &EmitterState()
   &SpawnBurst_Instantaneous(Use Spawn Probability=false, UseLoopCountLimit=false) {enabled=false}
   &SpawnRate(
    Spawn Probability=&FloatFromCurve(
     FloatCurve=DataInterface() {propertyValues {Curve {keys [
         {interpMode=RCIM_Cubic, leaveTangentWeight=0.7021, time=0.1, value=0}
         {interpMode=RCIM_Cubic, tangentMode=RCTM_User, time=0.55, value=1}
         {arriveTangentWeight=0.6283, interpMode=RCIM_Cubic, tangentMode=RCTM_User, time=1, value=0}
        ]}}}
     CurveIndex=&RemapRange(
      Input Value To Remap=&VectorLength(Vector=$Engine.Owner.Velocity)
      Input Min=0
      Input Max=200
      Output Min=0
      Output Max=1
     )
    )
    SpawnRate=20
    Use Spawn Probability=true
   )
  ]
  particleSpawnScript [
   &InitializeParticle(
    Color=&MergeHelper_ColorSwitch(
     0=<rgba 0.8878 1 0.7183 1>
     1=<rgba 0.6545 0.6573 0.4813 1>
     2=?
     3=?
     Index=&MakeIntFromFloat(Float=$User.VariantID)
     Num=::NewEnumerator0
    )
    Lifetime=&Multiply_Float(
     A=&RandomRangeFloat(Minimum=0.4, Maximum=0.6, Evaluation Type=::NewEnumerator1)
     B=&FloatFromCurve(
      FloatCurve=DataInterface() {propertyValues {Curve {keys [<at 0.5 1.25>,<at 1 0.5>]}}}
      CurveIndex=&FloatFromCurve(
       FloatCurve=DataInterface() {propertyValues {Curve {keys [
           {interpMode=RCIM_Cubic, leaveTangentWeight=0.7021, tangentMode=RCTM_User, time=0, value=0}
           {interpMode=RCIM_Cubic, tangentMode=RCTM_User, time=0.5, value=1}
           {arriveTangentWeight=0.6283, interpMode=RCIM_Cubic, tangentMode=RCTM_User, time=1, value=0}
          ]}}}
       CurveIndex=&RemapRange(
        Input Value To Remap=&VectorLength(Vector=$Engine.Owner.Velocity)
        Input Min=0
        Input Max=300
        Output Min=0
        Output Max=1
       )
      )
     )
    )
    Sprite Rotation Mode=::NewEnumerator2
    Sprite Size Mode=::NewEnumerator3
    Uniform Sprite Size=40
    UsePositionOffset=false
    Write Material Random=true
   )
   &ShapeLocation(
    MaskSpawnGroup=false
    Non Uniform Scale=<1 0.1 1>
    Offset=<0 0 -5>
    Offset Coordinate Space=::World
    Offset Mode=::NewEnumerator0
    Override Seed=false
    Sphere Radius=2
    UseOwnerScale=true
   )
  ]
  particleUpdateScript [
   &ParticleState()
   &SpriteRotationRate(Rotation Rate=&RandomRangeFloat(Minimum=-50, Maximum=50))
   &AccelerationForce(Acceleration=&RandomRangeVector(Minimum=<-15 -15 -175>, Maximum=<15 15 -125>, Evaluation Type=::NewEnumerator1))
   &BlendSpace("Position Blend (World <-> Local)"=0.9)
   &SolveForcesAndVelocity(Clamp Velocity=false, Limit Acceleration=false)
   &ScaleColor(
    Scale Alpha=&FloatFromCurve(FloatCurve=DataInterface() {propertyValues {Curve {keys [<at 0 0 RCIM_Cubic>,<at 0.2 1 RCIM_Cubic>,<at 1 0>]}}})
    ScaleA=true
    ScaleRGB=false
   )
   &ScaleSpriteSize(Scale Factor=&Vector2DFromFloat(Value=&FloatFromCurve(FloatCurve=DataInterface() {propertyValues {Curve {keys [{arriveTangent=1.095, interpMode=RCIM_Cubic, leaveTangent=1.095, tangentMode=RCTM_User, time=0, value=0.3}, <at 1 1>]}}})))
   &CameraOffset(Camera Offset Amount=-5)
   &SubUVAnimation(
    Random Start Frame=true
    "Sprite Renderer "=DataInterface() {propertyValues {SpriteRenderer=&self:PixieDust_1.NiagaraSpriteRendererProperties_0}}
    SubUV Lookup Index Scale:NiagaraFloat=&RandomRangeFloat(Minimum=0.4, Maximum=0.6)
    UseEndFrame=false
    UseStartFrame=false
   )
   &DynamicMaterialParameters(
    Index 0 Param 2=1
    Index 0 Param 3=-0.1
    Module.Index0Float0Write=true
    Module.Index0Float1Write=true
    Module.Index0Float2Write=true
    Module.Index0Float3Write=false
   )
   &KillParticles(Kill Particles=&SetBoolByFloatComparison(A=&VectorLength(Vector=$Engine.Owner.Velocity), B=1200))
   /SetVariables(Particles.NormalizedAge:NiagaraFloat=&MakeCustomFloatFromBool(
     Bool=&SetBoolByFloatComparison(A=&VectorLength(Vector=$Engine.Owner.Velocity), B=1200)
     True Float=&Multiply_Float(A=1.1, B=$Particles.NormalizedAge)
     False Float=$Particles.NormalizedAge
    )) {enabled=false, isSetParameters=true}
  ]
  rendererClasses [&NiagaraSpriteRendererProperties]
  renderers [{rendererClass=&NiagaraSpriteRendererProperties}]
 }
}
summary {
 systemName=NS_Guitar_GainSneeze_Idle_Sand
 emitters [
  {emitterName=BitsHigh, rendererClasses [&NiagaraSpriteRendererProperties]}
  {emitterName=PixieDust, rendererClasses [&NiagaraSpriteRendererProperties]}
  {bEnabled=false, emitterName=FogCenterHigh, rendererClasses [&NiagaraSpriteRendererProperties]}
  {bEnabled=false, emitterName=FogHigh, rendererClasses [&NiagaraSpriteRendererProperties]}
 ]
 userVariables [
  User.bIsFrontEnd {type=NiagaraBool}
  User.bIsFrontEndPreview {type=NiagaraBool}
  User.VariantID {type=NiagaraFloat, defaultValue=0, description=""}
 ]
}
