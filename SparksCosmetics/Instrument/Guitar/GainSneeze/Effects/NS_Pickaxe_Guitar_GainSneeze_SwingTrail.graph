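# Niagara swing trail for the GainSneeze guitar pickaxe cosmetic, built from two bone/socket-driven ribbon emitters.
# TrailSharp is a short-lived ribbon with curve-shaped width and a color shift over life; TrailSoft lingers slightly longer with a plain tint.
# User.VariantID selects each emitter's base color through a color switch.
# RGB is multiplied by a system ToDColor, and alpha combines curves, including emitter-age fades, with a time-of-day scalar that brightens at night.
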NiagaraSystem /SparksCosmetics/Instrument/Guitar/GainSneeze/Effects/NS_Pickaxe_Guitar_GainSneeze_SwingTrail
dependencies {
 usedDataInterfaces [&NiagaraDataInterfaceColorCurve, &NiagaraDataInterfaceCurve, &NiagaraDataInterfaceSocketReader]
 usedDynamicInputs []
 usedModules [
  &self:SystemSpawnScript.SystemScriptSource.SystemScriptGraph.NiagaraNodeAssignment_4.SetVariables_#1
  &SystemState
  &self:SystemSpawnScript.SystemScriptSource.SystemScriptGraph.NiagaraNodeAssignment_5.SetVariables_#2
 ]
}
emitters {
 TrailSharp {
  emitterSpawnScript [&FN_RibbonIDOffsetUtility()]
  emitterUpdateScript [
   &EmitterState()
   &FN_SpawnAnimTrail()
  ]
  particleSpawnScript [
   &FN_InitializeAnimTrailBoneIndecies(First Bone or Socket Index=1)
   &FN_InitializeAnimTrail(
    AnimTrailWidth=3
    Color=&MergeHelper_ColorSwitch(
     0=<rgba 1 1 1 0.5>
     1=<rgba 0.5281 0.4731 0.268 0.5>
     2=?
     3=?
     Index=&MakeIntFromFloat(Float=$User.VariantID)
     Num=::NewEnumerator0
    )
    Lifetime=0.18
   )
   &AddVelocity(Velocity=<800 0 0>) {enabled=false}
  ]
  particleUpdateScript [
   &SolveForcesAndVelocity(Clamp Velocity=false, Limit Acceleration=false)
   &FN_UpdateAnimTrail() {module=FN_UpdateAnimTrail001}
   &ScaleColor(
    Scale Alpha=&FloatFromCurve(
     FloatCurve=DataInterface() {propertyValues {Curve {keys [
         {arriveTangent=-9.8626, interpMode=RCIM_Cubic, leaveTangent=-9.8626, tangentMode=RCTM_User, time=0, value=2}
         {arriveTangent=-1.1117, interpMode=RCIM_Cubic, leaveTangent=-1.1117, tangentMode=RCTM_User, time=0.5, value=0.5}
         {interpMode=RCIM_Cubic, tangentMode=RCTM_User, time=1, value=0}
        ]}}}
     Scale Curve=&FloatFromCurve(
      FloatCurve=DataInterface() {propertyValues {Curve {keys [<at 0.1 1 RCIM_Cubic>,<at 0.2 0 RCIM_Cubic>]}}}
      CurveIndex=$Emitter.Age
      Scale Curve=&FNTimeOfDayScalar(Night Scale=1.3, Morning Scale=1, Evening Scale=1)
     )
    )
    Scale RGB=&MakeVectorFromLinearColorRGB(Color=&MultiplyLinearColor(
      Linear Color=&ColorFromCurve(ColorCurve=DataInterface() {propertyValues {
         AlphaCurve {keys [<at 0 1 RCIM_Cubic>]}
         BlueCurve {keys [<at 0 0 RCIM_Cubic>, <at 0.1435 0>, {interpMode=RCIM_Cubic, leaveTangent=2.5122, time=0.2853, value=1.3009}, <at 0.4416 1>]}
         GreenCurve {keys [<at 0 0 RCIM_Cubic>, <at 0.1435 1.2939>, {interpMode=RCIM_Cubic, leaveTangent=-0.7383, time=0.2853, value=2}, <at 0.4416 1>]}
         RedCurve {keys [<at 0 0.449 RCIM_Cubic>, <at 0.1435 2>, {interpMode=RCIM_Cubic, leaveTangent=-2.5122, time=0.2853, value=0}, <at 0.4416 1>]}
        }})
      Scale Factor=&MakeVector4_FromLinearColor(LinearColor=$System.ToDColor)
     ))
    ScaleA=true
    ScaleRGB=true
   )
   &ScaleRibbonWidth(Ribbon Width Scale=&FloatFromCurve(FloatCurve=DataInterface() {propertyValues {Curve {keys [<at 0 0.5>,<at 0.1 1>,<at 1 0.3>]}}}))
  ]
  rendererClasses [&NiagaraRibbonRendererProperties]
  renderers [{rendererClass=&NiagaraRibbonRendererProperties}]
 }
 TrailSoft {
  emitterSpawnScript [&FN_RibbonIDOffsetUtility()]
  emitterUpdateScript [
   &EmitterState()
   &FN_SpawnAnimTrail()
  ]
  particleSpawnScript [
   &FN_InitializeAnimTrailBoneIndecies()
   &FN_InitializeAnimTrail(
    Color=&MergeHelper_ColorSwitch(
     0=<rgba 0.6378 0.9323 0.6048 0.6>
     1=<rgba 0.9323 0.9035 0.5842 0.6>
     2=?
     3=?
     Index=&MakeIntFromFloat(Float=$User.VariantID)
     Num=::NewEnumerator0
    )
    Lifetime=0.3
   )
   &AddVelocity(Velocity=<800 0 0>) {enabled=false}
  ]
  particleUpdateScript [
   &SolveForcesAndVelocity(Clamp Velocity=false, Limit Acceleration=false)
   &FN_UpdateAnimTrail() {module=FN_UpdateAnimTrail001}
   &ScaleColor(
    Scale Alpha=&FloatFromCurve(
     FloatCurve=DataInterface() {propertyValues {Curve {keys [<at 0.07 1.3 RCIM_Cubic>,<at 0.17 0 RCIM_Cubic>]}}}
     CurveIndex=$Emitter.Age
     Scale Curve=&FloatFromCurve(
      FloatCurve=DataInterface() {propertyValues {Curve {keys [<at 0.1 1 RCIM_Cubic>, {arriveTangent=-1.9689, interpMode=RCIM_Cubic, leaveTangent=-1.9689, tangentMode=RCTM_User, time=0.6, value=0}]}}}
      CurveIndex=$Emitter.Age
      Scale Curve=&FNTimeOfDayScalar(Night Scale=1.3, Morning Scale=1, Evening Scale=1)
     )
    )
    Scale RGB=&MakeVectorFromLinearColorRGB(Color=$System.ToDColor)
    ScaleA=true
    ScaleRGB=true
   )
  ]
  rendererClasses [&NiagaraRibbonRendererProperties]
  renderers [{rendererClass=&NiagaraRibbonRendererProperties}]
 }
}
summary {
 systemName=NS_Pickaxe_Guitar_GainSneeze_SwingTrail
 emitters [
  {emitterName=TrailSharp, rendererClasses [&NiagaraRibbonRendererProperties]}
  {emitterName=TrailSoft, rendererClasses [&NiagaraRibbonRendererProperties]}
 ]
 userVariables [
  User.bIsFrontEnd {type=NiagaraBool}
  User.bIsFrontEndPreview {type=NiagaraBool}
  User.RibbonIDOffset {type=NiagaraInt32}
  User.VariantID {type=NiagaraFloat, defaultValue=0, description=""}
 ]
}
